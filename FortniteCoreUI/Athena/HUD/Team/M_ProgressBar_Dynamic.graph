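# M_ProgressBar_Dynamic is a translucent UI material for a progress bar in the Athena team HUD.
# A horizontal UV gradient is compared against a fill edge, interpolated from StartValue to EndValue by FillAmount.
# Up to that edge it blends ColorA to ColorB along the bar; beyond it, Unfilled Color applies, with color and alpha computed separately.
# A quest-progress collection parameter is present but unconnected.
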Material /FortniteCoreUI/Athena/HUD/Team/M_ProgressBar_Dynamic
graph {
 properties {
  bUsedWithStaticMesh=false
  blendMode=BLEND_Translucent
  materialDomain=MD_UI
  pixelDepthOffsetMode=PDOM_Legacy
  shadingModel=MSM_Unlit
  usedShadingModels=Unlit
 }
 nodes [
  CollectionParameter() {collection=/FortniteCoreUI/UI/InGame/Quest/AthenaQuestProgressFill, parameterName=Progress}
  n1 = MaterialFunctionCall(UV Channel=TextureCoordinate()) {materialFunction=&LinearGradient}
  n2 = LinearInterpolate(
   A=ScalarParameter() {parameterName=StartValue}
   B=ScalarParameter() {defaultValue=1, parameterName=EndValue}
   Alpha=ScalarParameter() {parameterName=FillAmount}
  )
  n3 = VectorParameter() {parameterName=Unfilled Color, defaultValue=<rgba 0 0 0 0.25>}
  n4 = VectorParameter() {parameterName=ColorA, defaultValue=<rgba 1 0 0 1>}
  n5 = VectorParameter() {parameterName=ColorB, defaultValue=<rgba 0 1 0 1>}
  n6 = LinearInterpolate(A=@n4.RGB, B=@n5.RGB, Alpha=@n1.UGradient)
  If(A=@n1.UGradient, B=@n2, "A > B"=@n3.RGB, "A == B"=@n6, "A < B"=@n6)
  n7 = LinearInterpolate(A=@n4.A, B=@n5.A, Alpha=@n1.UGradient)
  If(A=@n1.UGradient, B=@n2, "A > B"=@n3.A, "A == B"=@n7, "A < B"=@n7)
 ]
}
